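; Fortnite passive ability granting a tactical overshield that regenerates after a delay, with hotfixable tuning and mutator-driven enable checks.
; On activation it sets max overshield, shows the bar and binds shield listeners; damage or breaks remove regen and restart the delay.
; Regen uses a gameplay effect, or a timer-looped instant effect in spatial (Creative/UEFN) experiences.
; DBNO and bus transitions adjust regen, and overshield break cues wait a tick so normal shield breaks can cancel them.
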(event Custom|HandleAircraftStateChange
  ;[1 1 1] Handle getting onto the plane after warmup
  (Utilities|FlowControl|Delay 0.1)
  (bind _outfortplayerpawn (CallFunction|GetAvatarasValidFortPlayerPawn)
    (:ValidFortPlayerPawn
      (Utilities|IsinAircraft _outfortplayerpawn
        (:true
          (Utilities|Time|ClearandInvalidateTimerbyHandle (Variables|TimerVariables|GetTimer_RegenDelay)) ; Clear any timed regen that might be left over from warmup
          (Utilities|Time|ClearandInvalidateTimerbyHandle (Variables|TimerVariables|GetTimer_VariablePeriod)))
        (:false
          (bind _outfortplayerpawn_2 (CallFunction|GetAvatarasValidFortPlayerPawn)
            (:ValidFortPlayerPawn
              (if (Variables|ApplicationChecks|GetBool_ForceRechargeOnAircraftExit)
                (FortPawn|SetCurrentOvershield _outfortplayerpawn_2 0.0)
                (Utilities|FlowControl|Delay (ScalableFloat|GetValueatLevel (Variables|Rows|GetRow_BusDropRegenDelay)))
                (CallFunction|ApplyRegenNoDelay)
                (else
                  (FortPawn|SetCurrentOvershield _outfortplayerpawn_2 (FortPawn|GetMaxOvershield _outfortplayerpawn_2)))))))))
    (:AvatarCastFailed
      (Utilities|Time|ClearandInvalidateTimerbyHandle (Variables|TimerVariables|GetTimer_RegenDelay)) ; Clear any timed regen that might be left over from warmup
      (Utilities|Time|ClearandInvalidateTimerbyHandle (Variables|TimerVariables|GetTimer_VariablePeriod)))))

(event Custom|SetupAircraftBinds
  ;[1 1 1] Handle getting onto the plane after warmup
  (if (not (Variables|MiscVariables|GetbAircraftSetUp))
    (Variables|MiscVariables|SetbAircraftSetUp true)
    (bind _outfortplayerpawn (CallFunction|GetAvatarasValidFortPlayerPawn)
      (:ValidFortPlayerPawn
        (bind (_aircraftcomponent _targetpawnorcontrollerisvalid) (CallFunction|GetAircraftComponent _outfortplayerpawn))
        (if _targetpawnorcontrollerisvalid
          (Utilities|IsValid _aircraftcomponent
            (:Is%20Valid
              (EventDispatchers|BindEventtoOnAircraftStateChange _aircraftcomponent (delegate Custom|HandleAircraftStateChange)))))))))

(event Custom|ClearAircraftBinds
  ;[1 1 1] Handle getting onto the plane after warmup
  (if (Variables|MiscVariables|GetbAircraftSetUp)
    (Variables|MiscVariables|SetbAircraftSetUp false)
    (bind _outfortplayerpawn (CallFunction|GetAvatarasValidFortPlayerPawn)
      (:ValidFortPlayerPawn
        (bind (_aircraftcomponent _targetpawnorcontrollerisvalid) (CallFunction|GetAircraftComponent _outfortplayerpawn))
        (if _targetpawnorcontrollerisvalid
          (Utilities|IsValid _aircraftcomponent
            (:Is%20Valid
              (EventDispatchers|UnbindEventfromOnAircraftStateChange _aircraftcomponent (delegate Custom|HandleAircraftStateChange)))))))))

(event Custom|SetupDBNOBinds
  ;[1 1 1] Handle DBNO - When a player goes DBNO remove the effect. When they Revive re-apply the effect or wait for the delay based on a hotfixable value.
  (if (not (Variables|MiscVariables|GetbDBNOSetUp))
    (Variables|MiscVariables|SetbDBNOSetUp true)
    (Ability|Tasks|WaitGameplayTagAdd "(TagName=\"Gameplay.Status.DBNO\")"
      (:then
        (Ability|Tasks|WaitGameplayTagRemove "(TagName=\"Gameplay.Status.DBNO\")"
          (:Removed
            (Development|PrintString "Tac Overshield: DBNO End")
            (Utilities|IsinWarmUp
              (:false
                _)
              (:GameStateIsNotValid
                _))
            (if (Class|AthenaFunctionLibrary|RowtoBool (Variables|Rows|GetRow_RegenStartsOnDBNORevive))
              (Utilities|FlowControl|Delay (Math|Float|Clamp(Float) (ScalableFloat|GetValueatLevel (Variables|Rows|GetRow_DBNOReviveMinDelay)) 0.1 (ScalableFloat|GetValueatLevelwithContext (Variables|Rows|GetRow_TacticalOvershieldRegenDelay))))
              (CallFunction|ApplyRegenNoDelay)
              (else
                (CallFunction|ApplyDelayedRegen))))))
      (:Added
        (Development|PrintString "Tac Overshield: DBNO Start")
        (Utilities|Time|ClearandInvalidateTimerbyHandle (Variables|TimerVariables|GetTimer_RegenDelay))
        (CallFunction|AttemptRegenEffectRemoval)
        (bind _outfortplayerpawn (CallFunction|GetAvatarasValidFortPlayerPawn)
          (:ValidFortPlayerPawn
            (FortPawn|SetCurrentOvershield _outfortplayerpawn 0.0)))))))

(event Custom|OnVariablePeriodEnd
  ;[1 1 1] Reapply instant GE on variable-timed loop for Creative/UEFN
  (bind _bresult (CallFunction|TryApplyGameplayEffect (Variables|ShieldEffects|GetGE_ShieldRegenInstant)))
  (if _bresult
    (bind _returnvalue (Utilities|Time|SetTimerbyEvent (EventDispatchers|CreateEvent :!event Custom|OnVariablePeriodEnd) (ScalableFloat|GetValueatLevelwithContext (Variables|Rows|GetRow_TacticalOvershieldRegenPeriod))))
    (Variables|TimerVariables|SetTimer_VariablePeriod _returnvalue)))

(event Custom|Apply%20Delayed%20Regen
  ;[1 1 1] Apply Delay Regen - Wait and then apply the GE that gives overshield to the player on a tick. Also set up the shield check.
  (bind _returnvalue (Utilities|Time|SetTimerbyEvent (EventDispatchers|CreateEvent :!event Custom|ApplyRegenNoDelay) (Math|Float|Max(Float) (ScalableFloat|GetValueatLevelwithContext (Variables|Rows|GetRow_TacticalOvershieldRegenDelay)) 0.001)))
  (Variables|TimerVariables|SetTimer_RegenDelay _returnvalue))

(event Custom|ApplyRegenNoDelay
  ;[1 1 1] Apply Delay Regen - Wait and then apply the GE that gives overshield to the player on a tick. Also set up the shield check.
  (if (Fortnite|IsSpatialGameplayExperience)
    (bind _bresult (CallFunction|TryApplyGameplayEffect (Variables|ShieldEffects|GetGE_ShieldRegenInstant)))
    (if _bresult
      (bind _returnvalue (Utilities|Time|SetTimerbyEvent (EventDispatchers|CreateEvent :!event Custom|OnVariablePeriodEnd) (ScalableFloat|GetValueatLevelwithContext (Variables|Rows|GetRow_TacticalOvershieldRegenPeriod))))
      (Variables|TimerVariables|SetTimer_VariablePeriod _returnvalue))
    (else
      (bind _bresult_2 (CallFunction|TryApplyGameplayEffect (Variables|ShieldEffects|GetGE_ShieldRegen)))
      (if _bresult_2
        _)))
  (Ability|ExecuteGameplayCueOnOwner (Variables|GameplayCues|GetGC_Application)))

(event Custom|CheckMaxOverShield
  ;[1 1 1] Shield Check - Check the shield to see if we should remove the effect and play the full burst GC. Only play if the effect was already applied.
  (bind _overshieldfull (CallFunction|IsOvershieldFull (Variables|MiscVariables|GetPawnToCheck)))
  (if _overshieldfull
    (Utilities|AlreadyApplied
      (:True
        (Ability|ExecuteGameplayCueOnOwner (Variables|GameplayCues|GetGC_ShieldFull)))
      (:False
        _))
    (CallFunction|AttemptRegenEffectRemoval)))

(event Ability|EventActivateAbility
  ;[1 1 1] Activate Ability - Set up overshield then set up waits for the events and appy the delayed regen
  (|CallParentFunction)
  (Utilities|FlowControl|Sequence
    (:then_0
      (bind _outfortplayerpawn (CallFunction|GetAvatarasValidFortPlayerPawn)
        (:ValidFortPlayerPawn
          (bind _returnvalue (Fortnite|Utilities|AttemptGetPlayerStateforPawn _outfortplayerpawn))
          (FortPlayerState|BindEventtoOnMutatorListUpdatedDelegate _returnvalue (delegate Custom|OnMutatorListUpdated)))))
    (:then_1
      (bind _output_get (Variables|ApplicationChecks|SetOldIsEnabled (ScalableFloat|GetValueasBoolwithContext (Variables|Rows|GetRow_Enabled))))
      (if _output_get
        (Utilities|CheckApplicationTagRequirements
          (:true
            (bind _outfortplayerpawn_2 (CallFunction|GetAvatarasValidFortPlayerPawn)
              (:ValidFortPlayerPawn
                (bind _asynctaskproxy (Ability|Tasks|WaitGameplayEvent "(TagName=\"Event.TacOvershield.ApplyRegen\")")
                  (:then
                    (Variables|Default|SetWaitApplyRegenTask _asynctaskproxy)
                    (CallFunction|SetupDBNOBinds)
                    (CallFunction|SetupAircraftBinds)
                    (CallFunction|OverrideMaxOvershield _outfortplayerpawn_2)
                    (bind _asfort_player_state_zone (Utilities|Casting|CastToFortPlayerStateZone (Variables|/Script/Engine.Pawn|GetPlayerState _outfortplayerpawn_2))
                      (:then
                        (CallFunction|OverrideSetOvershieldVisibility _asfort_player_state_zone true))
                      (:CastFailed
                        (Development|PrintString "TacticalShield: could not get FortPlayerZone!" true true "(R=1.000000,G=0.355593,B=0.119792,A=1.000000)" 1e+01)))
                    (Class|FortOvershieldHelperComponent|ClearOvershieldListenerDelegates)
                    (Class|FortOvershieldHelperComponent|AssignOvershieldListenerDelegates (Utilities|Struct|MakeFortOvershieldDelegateContainer (delegate Custom|OnCurrentOvershieldChange) (delegate Custom|OnAnyShieldDamage) (delegate Custom|OnAnyShieldDamage) (delegate Custom|OnShieldBreak) (delegate Custom|OnOvershieldBreak) (delegate Custom|OnDamageReceived)))
                    (Utilities|IsinWarmUp
                      (:true
                        (bind _outfortplayerpawn_3 (CallFunction|GetAvatarasValidFortPlayerPawn)
                          (:ValidFortPlayerPawn
                            (FortPawn|SetCurrentOvershield _outfortplayerpawn_3 0.0))))
                      (:false
                        _)
                      (:GameStateIsNotValid
                        _))
                    (CallFunction|ApplyDelayedRegen))
                  (:EventReceived
                    (CallFunction|ApplyRegenNoDelay)))))))))))

(event Custom|OnCurrentOvershieldChange
  ;[1 1 1] Activate Ability - Set up overshield then set up waits for the events and appy the delayed regen
  (Utilities|IsinWarmUp
    (:false
      _)
    (:GameStateIsNotValid
      _))
  (CallFunction|CheckMaxOverShield))

(event Custom|OnAnyShieldDamage
  ;[1 1 1] Activate Ability - Set up overshield then set up waits for the events and appy the delayed regen
  (Utilities|IsinWarmUp
    (:false
      _)
    (:GameStateIsNotValid
      _))
  (CallFunction|AttemptRegenEffectRemoval))

(event Custom|OnShieldBreak
  (params (Instigator :type Actor%20Object%20Reference))
  ;[1 1 1] Activate Ability - Set up overshield then set up waits for the events and appy the delayed regen
  (Default|HandleNormalShieldBreak Instigator)
  (Utilities|IsinWarmUp
    (:false
      _)
    (:GameStateIsNotValid
      _))
  (CallFunction|AttemptRegenEffectRemoval))

(event Custom|OnMutatorListUpdated
  (bind _row_enabled (Variables|Rows|GetRow_Enabled))
  (bind _returnvalue (ScalableFloat|GetValueasBoolwithContext _row_enabled))
  ;[1 1 1] Activate Ability - Set up overshield then set up waits for the events and appy the delayed regen
  (if (!= (Variables|ApplicationChecks|GetOldIsEnabled) _returnvalue)
    (bind _output_get (Variables|ApplicationChecks|SetOldIsEnabled _returnvalue))
    (if (not _output_get)
      ;[1 1 1] On End Ability - Remove the effect and set max overshield to 0
      (Ability|RemoveGameplayEffectFromOwnerWithHandle (Variables|ShieldEffects|GetGE%20Overshield%20Regen))
      (CallFunction|ClearAircraftBinds)
      (bind _outfortplayerpawn (CallFunction|GetAvatarasValidFortPlayerPawn)
        (:ValidFortPlayerPawn
          (CallFunction|CleanupMaxOvershield _outfortplayerpawn)
          (bind _asfort_player_state_zone (Utilities|Casting|CastToFortPlayerStateZone (Variables|/Script/Engine.Pawn|GetPlayerState _outfortplayerpawn)))
          (CallFunction|OverrideSetOvershieldVisibility _asfort_player_state_zone))))
    (else
      (if _returnvalue
        _)))
  (Utilities|CheckApplicationTagRequirements
    (:true
      (bind _outfortplayerpawn_2 (CallFunction|GetAvatarasValidFortPlayerPawn)
        (:ValidFortPlayerPawn
          (bind _asynctaskproxy (Ability|Tasks|WaitGameplayEvent "(TagName=\"Event.TacOvershield.ApplyRegen\")")
            (:then
              (Variables|Default|SetWaitApplyRegenTask _asynctaskproxy)
              (CallFunction|SetupDBNOBinds)
              (CallFunction|SetupAircraftBinds)
              (CallFunction|OverrideMaxOvershield _outfortplayerpawn_2)
              (bind _asfort_player_state_zone_2 (Utilities|Casting|CastToFortPlayerStateZone (Variables|/Script/Engine.Pawn|GetPlayerState _outfortplayerpawn_2))
                (:then
                  (CallFunction|OverrideSetOvershieldVisibility _asfort_player_state_zone_2 true))
                (:CastFailed
                  (Development|PrintString "TacticalShield: could not get FortPlayerZone!" true true "(R=1.000000,G=0.355593,B=0.119792,A=1.000000)" 1e+01)))
              (Class|FortOvershieldHelperComponent|ClearOvershieldListenerDelegates)
              (Class|FortOvershieldHelperComponent|AssignOvershieldListenerDelegates (Utilities|Struct|MakeFortOvershieldDelegateContainer (delegate Custom|OnCurrentOvershieldChange) (delegate Custom|OnAnyShieldDamage) (delegate Custom|OnAnyShieldDamage) (delegate Custom|OnShieldBreak) (delegate Custom|OnOvershieldBreak) (delegate Custom|OnDamageReceived)))
              (Utilities|IsinWarmUp
                (:true
                  (bind _outfortplayerpawn_3 (CallFunction|GetAvatarasValidFortPlayerPawn)
                    (:ValidFortPlayerPawn
                      (FortPawn|SetCurrentOvershield _outfortplayerpawn_3 0.0))))
                (:false
                  _)
                (:GameStateIsNotValid
                  _))
              (CallFunction|ApplyDelayedRegen))
            (:EventReceived
              (CallFunction|ApplyRegenNoDelay))))))))

(event Custom|OnOvershieldBreak
  (params (Instigator :type Actor%20Object%20Reference))
  ;[1 1 1] Activate Ability - Set up overshield then set up waits for the events and appy the delayed regen
  (Default|HandleOvershieldBreak Instigator)
  (Utilities|IsinWarmUp
    (:false
      _)
    (:GameStateIsNotValid
      _))
  (CallFunction|AttemptRegenEffectRemoval))

(event Custom|OnDamageReceived
  (params (Payload :type Gameplay%20Event%20Data%20Structure))
  ;[1 1 1] Activate Ability - Set up overshield then set up waits for the events and appy the delayed regen
  (Utilities|StopRegenTrueDamageFilter Payload
    (:Pass
      (Utilities|IsinWarmUp
        (:false
          _)
        (:GameStateIsNotValid
          _))
      (CallFunction|AttemptRegenEffectRemoval))))

(event Ability|EventOnEndAbility
  (params (bWasCancelled :type Boolean))
  ;[1 1 1] On End Ability - Remove the effect and set max overshield to 0
  (|CallParentFunction bWasCancelled)
  (Utilities|FlowControl|Sequence
    (:then_0
      (Ability|RemoveGameplayEffectFromOwnerWithHandle (Variables|ShieldEffects|GetGE%20Overshield%20Regen))
      (CallFunction|ClearAircraftBinds)
      (bind _outfortplayerpawn (CallFunction|GetAvatarasValidFortPlayerPawn)
        (:ValidFortPlayerPawn
          (CallFunction|CleanupMaxOvershield _outfortplayerpawn)
          (bind _asfort_player_state_zone (Utilities|Casting|CastToFortPlayerStateZone (Variables|/Script/Engine.Pawn|GetPlayerState _outfortplayerpawn)))
          (CallFunction|OverrideSetOvershieldVisibility _asfort_player_state_zone))))
    (:then_1
      (bind _outfortplayerpawn_2 (CallFunction|GetAvatarasValidFortPlayerPawn)
        (:ValidFortPlayerPawn
          (bind _returnvalue (Fortnite|Utilities|AttemptGetPlayerStateforPawn _outfortplayerpawn_2))
          (Utilities|IsValid _returnvalue
            (:Is%20Valid
              (FortPlayerState|UnbindEventfromOnMutatorListUpdatedDelegate _returnvalue (EventDispatchers|CreateEvent :!event Custom|OnMutatorListUpdated)))))))
    (:then_2
      (GameplayTasks|EndTask (Variables|Default|GetWaitApplyRegenTask)))))

(event Ability|EventFailedToActivatePassiveAbility
  (params (ActorInfo :type Gameplay%20Ability%20Actor%20Info%20Structure))
  ;[1 1 1] Failed To Activate Passive Ability - Must have this parent call to avoid error
  (|CallParentFunction))

(event Custom|AttemptRegenEffectRemoval
  ;[1 1 1] Attempt Regen Effect Removal - Remove the regen effect if it is applied and then start the delay again to reapply the shield.
  (Utilities|Time|ClearandInvalidateTimerbyHandle (Variables|TimerVariables|GetTimer_VariablePeriod))
  (Utilities|AlreadyApplied
    (:True
      (Ability|RemoveGameplayEffectFromOwnerWithHandle (Variables|ShieldEffects|GetGE%20Overshield%20Regen) 1))
    (:False
      _))
  (CallFunction|ApplyDelayedRegen))

(fn CleanupMaxOvershield
  (params (PawnRef :type Fort%20Pawn%20Object%20Reference))
  (FortPawn|SetMaxOvershield PawnRef 0.0))

(fn GetAircraftComponent
  (params (TargetPawn :type Pawn%20Object%20Reference))
  (bind _returnvalue (Pawn|GetController TargetPawn))
  (bind _returnvalue_2 (Actor|GetComponentbyClass _returnvalue "/Script/FortniteGame.FortControllerComponent_Aircraft"))
  ;[1 1 1] Tries to get the aircraft component off of player controller
  (Utilities|IsValid TargetPawn
    (:Is%20Valid
      (Utilities|IsValid _returnvalue
        (:Is%20Valid
          (Utilities|IsValid _returnvalue_2
            (:Is%20Valid
              (return _returnvalue_2 true))
            (:Is%20Not%20Valid
              (return 0 true))))
        (:Is%20Not%20Valid
          _)))
    (:Is%20Not%20Valid
      _))
  (return 0 false))

(fn GetValidPawn
  (bind _pawntocheck (Variables|MiscVariables|GetPawnToCheck))
  ;[1 1 1] Save out a valid pawn ref to the ability owner so we don't cast every time we need to do a get off of it. This is needed for things like the jumping out of the bus which resets your pawn
  (Utilities|IsValid _pawntocheck
    (:Is%20Valid
      (return _pawntocheck true))
    (:Is%20Not%20Valid
      (bind _outfortplayerpawn (CallFunction|GetAvatarasValidFortPlayerPawn)
        (:ValidFortPlayerPawn
          (bind _output_get (Variables|MiscVariables|SetPawnToCheck _outfortplayerpawn))
          (return _output_get true))
        (:AvatarCastFailed
          (return 0 false))))))

(fn HandleNormalShieldBreak
  (params (EffectCauser :type Actor%20Object%20Reference))
  ;[1 1 1] Cancel our overshield break if the shield breaks too
  (if (Variables|BreakCueManagement|GetBool_HasPendingOvershieldBreak)
    (if (== EffectCauser (Variables|BreakCueManagement|GetPendingOvershieldBreaker))
      (Utilities|Time|ClearandInvalidateTimerbyHandle (Variables|BreakCueManagement|GetOvershieldBreakTimerHandle)))))

(fn HandleOvershieldBreak
  (params (EffectCauser :type Actor%20Object%20Reference))
  ;[1 1 1] If the overshield breaks play a cue.
  (bind (_pawntocheck _passedpawncast) (CallFunction|GetValidPawn))
  (if _passedpawncast
    (if (<= (FortPawn|GetShield _pawntocheck) 0)
      (Variables|BreakCueManagement|SetBool_HasPendingOvershieldBreak false)
      (Ability|ExecuteGameplayCueWithParamsOnOwner (Variables|GameplayCues|GetGC_NormalShieldBreak) (Ability|GameplayCue|MakeGameplayCueParameters :NormalizedMagnitude 0.0 :RawMagnitude 0.0 :Location "0,0,0" :Normal "0,0,0" :EffectCauser EffectCauser :GameplayEffectLevel 1 :AbilityLevel 1 :bReplicateLocationWhenUsingMinimalRepProxy false))
      (else
        (Variables|BreakCueManagement|SetPendingOvershieldBreaker EffectCauser)
        (Variables|BreakCueManagement|SetBool_HasPendingOvershieldBreak true)
        (bind _returnvalue (Utilities|Time|SetTimerforNextTickbyEvent (EventDispatchers|CreateEvent :!event SendOvershieldBreakCue)))
        (Variables|BreakCueManagement|SetOvershieldBreakTimerHandle _returnvalue)))))

(fn IsOvershieldFull
  (params (OptionalPawnIn :type Fort%20Pawn%20Object%20Reference))
  ;[1 1 1] Checks if the overshield is full, if no pawn is passed in, it does a get avatar as valid actor check.
  (Utilities|IsValid OptionalPawnIn
    (:Is%20Valid
      (return (>= (FortPawn|GetCurrentOvershield OptionalPawnIn) (* (FortPawn|GetMaxOvershield OptionalPawnIn) (ScalableFloat|GetValueatLevel (Variables|Rows|GetRow_TacticalOvershieldRegenCapPercentage))))))
    (:Is%20Not%20Valid
      (bind _outfortplayerpawn (CallFunction|GetAvatarasValidFortPlayerPawn)
        (:ValidFortPlayerPawn
          (return (>= (FortPawn|GetCurrentOvershield _outfortplayerpawn) (* (FortPawn|GetMaxOvershield _outfortplayerpawn) (ScalableFloat|GetValueatLevel (Variables|Rows|GetRow_TacticalOvershieldRegenCapPercentage))))))
        (:AvatarCastFailed
          (return true))))) ; If Valid Pawn passed in use that to check
)

(fn OverrideMaxOvershield
  (params (PawnRef :type Fort%20Pawn%20Object%20Reference))
  (FortPawn|SetMaxOvershield PawnRef (Math|Float|SelectFloat (ScalableFloat|GetValueatLevelwithContext (Variables|Rows|GetRow_MaxTacticalOvershield)) 0.0 (ScalableFloat|GetValueasBool (Variables|Rows|GetRow_Enabled)))))

(fn OverrideSetOvershieldVisibility
  (params (PlayerStateRef :type Fort%20Player%20State%20Zone%20Object%20Reference) (Visible :type Boolean))
  (Class|FortPlayerStateZone|SetOvershieldBarVisible PlayerStateRef Visible))

(fn SendOvershieldBreakCue
  (bind (_pawntocheck _passedpawncast) (CallFunction|GetValidPawn))
  (if _passedpawncast
    (if (<= (FortPawn|GetCurrentOvershield _pawntocheck) 0)
      (Ability|ExecuteGameplayCueWithParamsOnOwner (Variables|GameplayCues|GetGC_TacOvershieldBreak) (Ability|GameplayCue|MakeGameplayCueParameters :NormalizedMagnitude 0.0 :RawMagnitude 0.0 :Location "0,0,0" :Normal "0,0,0" :Instigator _pawntocheck :EffectCauser (Variables|BreakCueManagement|GetPendingOvershieldBreaker) :GameplayEffectLevel 1 :AbilityLevel 1 :bReplicateLocationWhenUsingMinimalRepProxy false))
      (Variables|BreakCueManagement|SetPendingOvershieldBreaker 0))))

(fn TryApplyGameplayEffect
  (params (GameplayEffect :type Gameplay%20Effect%20Class%20Reference))
  (Utilities|CheckApplicationBlocked
    (:Allowed
      (bind _overshieldfull (CallFunction|IsOvershieldFull 0))
      (if (not _overshieldfull)
        (Utilities|Time|ClearandInvalidateTimerbyHandle (Variables|TimerVariables|GetTimer_RegenDelay))
        (Utilities|AlreadyApplied
          (:True
            _)
          (:False
            (bind _returnvalue (Ability|GameplayEffect|AssignTagSetbyCallerMagnitude (Ability|MakeOutgoingGameplayEffectSpec GameplayEffect) "(TagName=\"SetByCaller.Magnitude\")" (ScalableFloat|GetValueatLevelwithContext (Variables|Rows|GetRow_TacticalOvershieldRegenAmount))))
            (bind _returnvalue_2 (Ability|ApplyGameplayEffectSpecToOwner _returnvalue))
            (Variables|ShieldEffects|SetGE%20Overshield%20Regen _returnvalue_2)
            (return true))))
      (return false))
    (:Blocked
      _)))
